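# Placeholder SoundCue in NitroDrome's TEMP audio folder, named for a vehicle-entered sound.
# The node graph is empty with no root node, so the cue references no audio and plays nothing.
# It is assigned to the Fort_Traps sound class, with no attenuation asset shown.
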SoundCue /NitroDrome/Assets/TEMP/Audio/NitroDrome_Vehicle_Entered_Cue
root=null
nodes []
properties {maxDistance=2097152, soundClassObject=&Fort_Traps, volumeMultiplier=1}
